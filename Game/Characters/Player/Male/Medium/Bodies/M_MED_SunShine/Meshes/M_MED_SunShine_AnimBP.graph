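# Animation Blueprint adding secondary motion to the SunShine male body's jacket, collar, grenade belt, straps, kneepad and pins.
# It copies the parent mesh pose, then layers weighted bone constraints, Control Rig, jacket look-ats and offsets, RigidBody simulation and AnimDynamics.
# Each update sets RigidBody gravity through a deprecated helper and derives correction alphas from skydive, crouch, emote and rigid-body state, easing most with a fixed delta time.
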AnimBlueprint /Game/Characters/Player/Male/Medium/Bodies/M_MED_SunShine/Meshes/M_MED_SunShine_AnimBP
target_skeleton=&/M_MED_SunShine_Skeleton
graphs [
 AnimGraph {nodes [
   n1 = Anim Dynamics - Bone: dyn_collar_back_r_01!(ComponentPose, Alpha=1, GravityOverride:Vector="0,0,-980", ExternalForce="0,0,-980") -> (Pose) {
    angularDampingOverride=0.9999
    angularSpringConstant=100
    bAngularSpring=true
    bOverrideAngularDamping=true
    bUseGravityOverride=true
    simulationSpace=BoneRelative
    alphaBoolBlend {blendOption=Linear}
    boundBone {boneName=dyn_collar_back_r_01}
    componentAppliedLinearAccClamp=<1500 1500 1500>
    componentLinearAccScale=<0.6 0.6 0.8>
    componentLinearVelScale=<0.6 0.6 0.8>
    physicsBodyDefinitions [{
      boundBone {boneName=dyn_collar_back_r_01}
      boxExtents=<10 10 10>
      constraintSetup {angularLimitsMax=<2.5 0 2.5>, angularLimitsMin=<-2.5 -10 -2.5>, angularTarget=<0 0 0>}
      localJointOffset=<1 0 1>
     }]
    relativeSpaceBone {boneName=pelvis}
   }
   Anim Dynamics - Bone: dyn_collar_back_l_01!(ComponentPose=@n1, Alpha=1, GravityOverride:Vector="0,0,-980", ExternalForce="0,0,-980") -> (Pose) {
    angularDampingOverride=0.9999
    angularSpringConstant=100
    bAngularSpring=true
    bOverrideAngularDamping=true
    bUseGravityOverride=true
    simulationSpace=BoneRelative
    alphaBoolBlend {blendOption=Linear}
    boundBone {boneName=dyn_collar_back_l_01}
    componentAppliedLinearAccClamp=<1500 1500 1500>
    componentLinearAccScale=<0.6 0.6 0.8>
    componentLinearVelScale=<0.6 0.6 0.8>
    physicsBodyDefinitions [{
      boundBone {boneName=dyn_collar_back_l_01}
      boxExtents=<10 10 10>
      constraintSetup {angularLimitsMax=<2.5 0 2.5>, angularLimitsMin=<-2.5 -10 -2.5>, angularTarget=<0 0 0>}
      localJointOffset=<1 0 1>
     }]
    relativeSpaceBone {boneName=pelvis}
   }
   n2 = Anim Dynamics - Bone: dyn_collar_front_r!(ComponentPose, Alpha=1, GravityOverride:Vector="0,0,-980", ExternalForce="0,0,-980") -> (Pose) {
    angularDampingOverride=0.9999
    angularSpringConstant=100
    bAngularSpring=true
    bOverrideAngularDamping=true
    bUseGravityOverride=true
    simulationSpace=BoneRelative
    alphaBoolBlend {blendOption=Linear}
    boundBone {boneName=dyn_collar_front_r}
    componentAppliedLinearAccClamp=<1500 1500 1500>
    componentLinearAccScale=<0.6 0.6 0.8>
    componentLinearVelScale=<0.6 0.6 0.8>
    physicsBodyDefinitions [{
      boundBone {boneName=dyn_collar_front_r}
      boxExtents=<10 10 10>
      constraintSetup {angularLimitsMax=<5 0 5>, angularLimitsMin=<-5 -10 -5>, angularTarget=<0 0 0>}
      localJointOffset=<1 0 1>
     }]
    relativeSpaceBone {boneName=pelvis}
   }
   Anim Dynamics - Bone: dyn_collar_front_l!(ComponentPose=@n2, Alpha=1, GravityOverride:Vector="0,0,-980", ExternalForce="0,0,-980") -> (Pose) {
    angularDampingOverride=0.9999
    angularSpringConstant=100
    bAngularSpring=true
    bOverrideAngularDamping=true
    bUseGravityOverride=true
    simulationSpace=BoneRelative
    alphaBoolBlend {blendOption=Linear}
    boundBone {boneName=dyn_collar_front_l}
    componentAppliedLinearAccClamp=<1500 1500 1500>
    componentLinearAccScale=<0.6 0.6 0.8>
    componentLinearVelScale=<0.6 0.6 0.8>
    physicsBodyDefinitions [{
      boundBone {boneName=dyn_collar_front_l}
      boxExtents=<10 10 10>
      constraintSetup {angularLimitsMax=<5 0 5>, angularLimitsMin=<-5 -10 -5>, angularTarget=<0 0 0>}
      localJointOffset=<1 0 -1>
     }]
    relativeSpaceBone {boneName=pelvis}
   }
   "Transform (Modify) Bone - Bone: dyn_jacket_l_bk_master"(Alpha=1, Translation="0,0,0", Rotation="0,0,30", Scale="1,1,1") {
    rotationMode=BMM_Additive
    rotationSpace=BCS_BoneSpace
    translationMode=BMM_Additive
    translationSpace=BCS_BoneSpace
    alphaBoolBlend {blendOption=Linear}
    boneToModify {boneName=dyn_jacket_l_bk_master}
    rotation {roll=30}
   }
   n3 = "Constraint - dyn_collar_l_mstr (PPPPPP)"!(
    ComponentPose=Local To Component(LocalPose=Copy Pose From Mesh() {bCopyCurves=true, bCopyCustomAttributes=true, bUseAttachedParent=true, bUseMeshPose=true})
    Alpha=1
    ConstraintWeights_0=0.0004
    ConstraintWeights_1:float=0.0112
    ConstraintWeights_2:float=0.5941
    ConstraintWeights_3:float=0.3228
    ConstraintWeights_4:float=0.0191
    ConstraintWeights_5:float=0.0523
   ) -> (Pose) {
    alphaBoolBlend {blendOption=Linear}
    boneToModify {boneName=dyn_collar_l_mstr}
    constraintSetup [
     {targetBone {boneName=spine_03}}
     {targetBone {boneName=spine_04}}
     {targetBone {boneName=spine_05}}
     {targetBone {boneName=clavicle_l}}
     {targetBone {boneName=neck_01}}
     {targetBone {boneName=neck_02}}
    ]
    constraintWeights [1,1,1,1,1,1]
   }
   n4 = "Constraint - dyn_collar_r_mstr (PPPPPPPPPPP)"!(
    ComponentPose=@n3
    Alpha=1
    ConstraintWeights_0=0
    ConstraintWeights_1:float=0.0065
    ConstraintWeights_2:float=0.0225
    ConstraintWeights_3:float=0.3699
    ConstraintWeights_4:float=0.0035
    ConstraintWeights_5:float=0.0984
    ConstraintWeights_6:float=0.0748
    ConstraintWeights_7:float=0.103
    ConstraintWeights_8:float=0.316
    ConstraintWeights_9:float=0.0052
    ConstraintWeights_10:float=0.0001
   ) -> (Pose) {
    alphaBoolBlend {blendOption=Linear}
    boneToModify {boneName=dyn_collar_r_mstr}
    constraintSetup [
     {targetBone {boneName=spine_02}}
     {targetBone {boneName=spine_03}}
     {targetBone {boneName=spine_04}}
     {targetBone {boneName=spine_05}}
     {targetBone {boneName=clavicle_l}}
     {targetBone {boneName=clavicle_r}}
     {targetBone {boneName=neck_01}}
     {targetBone {boneName=neck_02}}
     {targetBone {boneName=head}}
     {targetBone {boneName=dyn_collar_front_r}}
     {targetBone {boneName=dyn_collar_front_l}}
    ]
    constraintWeights [1,1,1,1,1,1,1,1,1,1,1]
   }
   n5 = "Constraint - dyn_collar_fr_mstr (PPPPPPPP)"!(
    ComponentPose=@n4
    Alpha=1
    ConstraintWeights_0=0.0007
    ConstraintWeights_1:float=0.0235
    ConstraintWeights_2:float=0.0829
    ConstraintWeights_3:float=0.674
    ConstraintWeights_4:float=0.1166
    ConstraintWeights_5:float=0.0295
    ConstraintWeights_6:float=0.0719
    ConstraintWeights_7:float=0.001
   ) -> (Pose) {
    alphaBoolBlend {blendOption=Linear}
    boneToModify {boneName=dyn_collar_fr_mstr}
    constraintSetup [
     {targetBone {boneName=spine_02}}
     {targetBone {boneName=spine_03}}
     {targetBone {boneName=spine_04}}
     {targetBone {boneName=spine_05}}
     {targetBone {boneName=clavicle_l}}
     {targetBone {boneName=neck_01}}
     {targetBone {boneName=neck_02}}
     {targetBone {boneName=head}}
    ]
    constraintWeights [1,1,1,1,1,1,1,1]
   }
   n6 = "Constraint - dyn_collar_bk_mstr (PPPPPPPPPP)"!(
    ComponentPose=@n5
    Alpha=1
    ConstraintWeights_0=0.0175
    ConstraintWeights_1:float=0.0054
    ConstraintWeights_2:float=0.5952
    ConstraintWeights_3:float=0.0311
    ConstraintWeights_4:float=0.0158
    ConstraintWeights_5:float=0.0881
    ConstraintWeights_6:float=0.2006
    ConstraintWeights_7:float=0.0461
    ConstraintWeights_8:float=0.0001
    ConstraintWeights_9:float=0.0001
   ) -> (Pose) {
    alphaBoolBlend {blendOption=Linear}
    boneToModify {boneName=dyn_collar_bk_mstr}
    constraintSetup [
     {targetBone {boneName=spine_03}}
     {targetBone {boneName=spine_04}}
     {targetBone {boneName=spine_05}}
     {targetBone {boneName=clavicle_l}}
     {targetBone {boneName=clavicle_r}}
     {targetBone {boneName=neck_01}}
     {targetBone {boneName=neck_02}}
     {targetBone {boneName=head}}
     {targetBone {boneName=dyn_collar_l}}
     {targetBone {boneName=dyn_collar_r}}
    ]
    constraintWeights [1,1,1,1,1,1,1,1,1,1]
   }
   n7 = "Constraint - dyn_grenade_belt_upper_mstr (PPPPP)"!(
    ComponentPose=@n6
    Alpha=1
    ConstraintWeights_0=0.8927
    ConstraintWeights_1:float=0.0484
    ConstraintWeights_2:float=0.0527
    ConstraintWeights_3:float=0.0007
    ConstraintWeights_4:float=0.0055
   ) -> (Pose) {
    alphaBoolBlend {blendOption=Linear}
    boneToModify {boneName=dyn_grenade_belt_upper_mstr}
    constraintSetup [
     {targetBone {boneName=pelvis}}
     {targetBone {boneName=spine_01}}
     {targetBone {boneName=spine_02}}
     {targetBone {boneName=thigh_twist_01_l}}
     {targetBone {boneName=deform_glute_l}}
    ]
    constraintWeights [1,1,1,1,1]
   }
   n8 = "Transform (Modify) Bone - Bone: dyn_grenade_belt_upper"(ComponentPose=@n7, Alpha=1, Translation="0,0,0.5", Rotation="0,0,0", Scale="1,1,1") {
    rotationMode=BMM_Additive
    rotationSpace=BCS_BoneSpace
    translationMode=BMM_Additive
    translationSpace=BCS_BoneSpace
    alphaBoolBlend {blendOption=Linear}
    boneToModify {boneName=dyn_grenade_belt_upper}
    translation {z=0.5}
   }
   n9 = "Constraint - dyn_strap_fr_adjust_mstr (PPPPPPPPPPP)"!(
    ComponentPose=@n8
    Alpha=1
    ConstraintWeights_0=0.7244
    ConstraintWeights_1:float=0.0844
    ConstraintWeights_2:float=0.0985
    ConstraintWeights_3:float=0.0575
    ConstraintWeights_4:float=0.0001
    ConstraintWeights_5:float=0.0157
    ConstraintWeights_6:float=0.0002
    ConstraintWeights_7:float=0.0178
    ConstraintWeights_8:float=0.001
    ConstraintWeights_9:float=0
    ConstraintWeights_10:float=0.0004
   ) -> (Pose) {
    alphaBoolBlend {blendOption=Linear}
    boneToModify {boneName=dyn_strap_fr_adjust_mstr}
    constraintSetup [
     {targetBone {boneName=pelvis}}
     {targetBone {boneName=spine_01}}
     {targetBone {boneName=spine_02}}
     {targetBone {boneName=spine_03}}
     {targetBone {boneName=spine_04}}
     {targetBone {boneName=thigh_twist_01_l}}
     {targetBone {boneName=deform_glute_l}}
     {targetBone {boneName=deform_groin_l}}
     {targetBone {boneName=thigh_twist_01_r}}
     {targetBone {boneName=deform_glute_r}}
     {targetBone {boneName=deform_groin_r}}
    ]
    constraintWeights [1,1,1,1,1,1,1,1,1,1,1]
   }
   n10 = "Transform (Modify) Bone - Bone: dyn_strap_fr_adjust"(ComponentPose=@n9, Alpha=1, Translation="0,0,0.5", Rotation="0,0,0", Scale="1,1,1") {
    rotationMode=BMM_Additive
    rotationSpace=BCS_BoneSpace
    translationMode=BMM_Additive
    translationSpace=BCS_BoneSpace
    alphaBoolBlend {blendOption=Linear}
    boneToModify {boneName=dyn_strap_fr_adjust}
    translation {z=0.5}
   }
   n11 = Control Rig!(
    Source=Component To Local(ComponentPose=@n10)
    Alpha
    crouch_alpha:double=Get crouch_alpha() -> (crouch_alpha:double)
    emote_alpha:double=Get emote alpha() -> (emote alpha:double)
   ) -> (Pose) {
    alphaBoolBlend {blendOption=Linear}
    controlRigAssetReference {blueprintRigClass=&/M_MED_SunShine_CtrlRig.M_MED_SunShine_CtrlRig_C}
   }
   n12 = "Transform (Modify) Bone - Bone: pelvis_target_back"(
    ComponentPose=Local To Component(LocalPose=@n11)
    Alpha=Get back_orient_out() -> (back_orient_out:double)
    Translation="0,-10,0"
    Rotation="0,0,0"
    Scale="1,1,1"
   ) {
    rotationMode=BMM_Additive
    rotationSpace=BCS_BoneSpace
    translationMode=BMM_Additive
    translationSpace=BCS_BoneSpace
    alphaBoolBlend {blendOption=Linear}
    boneToModify {boneName=pelvis_target_back}
    translation {y=-10}
   }
   n13 = "Transform (Modify) Bone - Bone: pelvis_target_back"(ComponentPose=@n12, Alpha=Get back_in() -> (back_in:double), Translation="0,2,0", Rotation="0,0,0", Scale="1,1,1") {
    rotationMode=BMM_Additive
    rotationSpace=BCS_BoneSpace
    translationMode=BMM_Additive
    translationSpace=BCS_BoneSpace
    alphaBoolBlend {blendOption=Linear}
    boneToModify {boneName=pelvis_target_back}
    translation {y=2}
   }
   n14 = Look At - Bone: dyn_jacket_l_fr_orient(ComponentPose=@n13, Alpha=1) {
    alphaBoolBlend {blendOption=Linear}
    boneToModify {boneName=dyn_jacket_l_fr_orient}
    lookAtLocation {x=0}
    lookAtTarget {boneReference {boneName=pelvis_target_fr_l}}
    lookAt_Axis {axis=<1 0>}
    lookUp_Axis {axis {x=0, z=1}}
   }
   n15 = Look At - Bone: dyn_jacket_r_fr_orient(ComponentPose=@n14, Alpha=1) {
    alphaBoolBlend {blendOption=Linear}
    boneToModify {boneName=dyn_jacket_r_fr_orient}
    lookAtLocation {x=0}
    lookAtTarget {boneReference {boneName=pelvis_target_fr_r}}
    lookAt_Axis {axis=<1 0>}
    lookUp_Axis {axis {x=0, z=1}}
   }
   n16 = Look At - Bone: dyn_jacket_l_bk_orient(ComponentPose=@n15, Alpha=1) {
    alphaBoolBlend {blendOption=Linear}
    boneToModify {boneName=dyn_jacket_l_bk_orient}
    lookAtLocation {x=0}
    lookAtTarget {boneReference {boneName=pelvis_target_back}}
    lookAt_Axis {axis=<1 0>}
    lookUp_Axis {axis {x=0, z=1}}
   }
   n17 = Look At - Bone: dyn_jacket_r_bk_orient(ComponentPose=@n16, Alpha=1) {
    alphaBoolBlend {blendOption=Linear}
    boneToModify {boneName=dyn_jacket_r_bk_orient}
    lookAtLocation {x=0}
    lookAtTarget {boneReference {boneName=pelvis_target_back}}
    lookAt_Axis {axis=<1 0>}
    lookUp_Axis {axis {x=0, z=1}}
   }
   n18 = "Transform (Modify) Bone - Bone: dyn_jacket_r_bk_master"(ComponentPose=@n17, Alpha=1, Translation="0,0,0", Rotation="0,0,-30", Scale="1,1,1") {
    rotationMode=BMM_Additive
    rotationSpace=BCS_BoneSpace
    translationMode=BMM_Additive
    translationSpace=BCS_BoneSpace
    alphaBoolBlend {blendOption=Linear}
    boneToModify {boneName=dyn_jacket_r_bk_master}
    rotation {roll=-30}
   }
   n19 = "Transform (Modify) Bone - Bone: dyn_jacket_l_bk_01"(ComponentPose=@n18, Alpha=1, Translation="0,0,0", Rotation="0,0,15", Scale="1,1,1") {
    rotationMode=BMM_Additive
    rotationSpace=BCS_BoneSpace
    translationMode=BMM_Additive
    translationSpace=BCS_BoneSpace
    alphaBoolBlend {blendOption=Linear}
    boneToModify {boneName=dyn_jacket_l_bk_01}
    rotation {roll=15}
   }
   n20 = "Transform (Modify) Bone - Bone: dyn_jacket_l_01"(ComponentPose=@n19, Alpha=1, Translation="0,0,0", Rotation="-15,0,0", Scale="1,1,1") {
    rotationMode=BMM_Additive
    rotationSpace=BCS_BoneSpace
    translationMode=BMM_Additive
    translationSpace=BCS_BoneSpace
    alphaBoolBlend {blendOption=Linear}
    boneToModify {boneName=dyn_jacket_l_01}
    rotation {pitch=-15}
   }
   n21 = "Transform (Modify) Bone - Bone: dyn_grenade_belt_upper"(ComponentPose=@n20, Alpha=1, Translation="0,0,0.5", Rotation="0,0,0", Scale="1,1,1") {
    rotationMode=BMM_Additive
    rotationSpace=BCS_BoneSpace
    translationMode=BMM_Additive
    translationSpace=BCS_BoneSpace
    alphaBoolBlend {blendOption=Linear}
    boneToModify {boneName=dyn_grenade_belt_upper}
    translation {z=0.5}
   }
   n22 = Get GravityOverride()
   n23 = RigidBody!(ComponentPose=@n21, Alpha, OverrideWorldGravity:Vector=@n22, ExternalForce=@n22) -> (Pose) {
    bClampLinearTranslationLimitToRefPose=true
    bOverrideWorldGravity=true
    overridePhysicsAsset=&/M_MED_SunShine_Physics
    simulationSpace=BaseBoneSpace
    alphaBoolBlend {blendOption=Linear}
    baseBoneRef {boneName=pelvis}
    componentAppliedLinearAccClamp=<1500 1500 1500>
    componentLinearAccScale=<0.35 0.35 0.45>
    overrideWorldGravity {z=-980}
    simSpaceSettings {
     maxAngularAcceleration=100
     maxAngularVelocity=3
     maxLinearAcceleration=700
     maxLinearVelocity=500
     velocityScaleZ=0
     worldAlpha=1
     externalLinearDragV=<0.4 0.4 0.4>
    }
   }
   n24 = "Constraint - dyn_kneepad_mstr (PPPPPPPPPPPP)"!(
    ComponentPose=@n23
    Alpha=1
    ConstraintWeights_0=0.0002
    ConstraintWeights_1:float=0
    ConstraintWeights_2:float=0
    ConstraintWeights_3:float=0.3346
    ConstraintWeights_4:float=0.2877
    ConstraintWeights_5:float=0
    ConstraintWeights_6:float=0.0339
    ConstraintWeights_7:float=0.2977
    ConstraintWeights_8:float=0.0174
    ConstraintWeights_9:float=0.0281
    ConstraintWeights_10:float=0
    ConstraintWeights_11:float=0.0004
   ) -> (Pose) {
    alphaBoolBlend {blendOption=Linear}
    boneToModify {boneName=dyn_kneepad_mstr}
    constraintSetup [
     {targetBone {boneName=pelvis}}
     {targetBone {boneName=thigh_twist_01_l}}
     {targetBone {boneName=deform_groin_l}}
     {targetBone {boneName=thigh_r}}
     {targetBone {boneName=calf_r}}
     {targetBone {boneName=calf_twist_01_r}}
     {targetBone {boneName=calf_twist_02_r}}
     {targetBone {boneName=deform_knee_out_r}}
     {targetBone {boneName=deform_knee_in_r}}
     {targetBone {boneName=thigh_twist_01_r}}
     {targetBone {boneName=deform_glute_r}}
     {targetBone {boneName=deform_groin_r}}
    ]
    constraintWeights [1,1,1,1,1,1,1,1,1,1,1,1]
   }
   n25 = "Constraint - dyn_pin_r_mstr (PPP)"!(ComponentPose=@n24, Alpha=1, ConstraintWeights_0=0.64, ConstraintWeights_1:float=0.2143, ConstraintWeights_2:float=0.1457) -> (Pose) {
    alphaBoolBlend {blendOption=Linear}
    boneToModify {boneName=dyn_pin_r_mstr}
    constraintSetup [{targetBone {boneName=calf_twist_01_r}}, {targetBone {boneName=calf_twist_02_r}}, {targetBone {boneName=foot_r}}]
    constraintWeights [1,1,1]
   }
   n26 = "Constraint - dyn_pin_l_mstr (PPP)"!(ComponentPose=@n25, Alpha=1, ConstraintWeights_0=0.64, ConstraintWeights_1:float=0.2143, ConstraintWeights_2:float=0.1457) -> (Pose) {
    alphaBoolBlend {blendOption=Linear}
    boneToModify {boneName=dyn_pin_l_mstr}
    constraintSetup [{targetBone {boneName=calf_twist_01_l}}, {targetBone {boneName=calf_twist_02_l}}, {targetBone {boneName=foot_l}}]
    constraintWeights [1,1,1]
   }
   n27 = Anim Dynamics - Bone: dyn_collar_front_r!(ComponentPose=@n26, Alpha=1, ExternalForce="0,0,0") -> (Pose) {
    angularDampingOverride=1
    angularSpringConstant=5000
    bAngularSpring=true
    bGravityOverrideInSimSpace=true
    bOverrideAngularDamping=true
    bUseGravityOverride=true
    simulationSpace=BoneRelative
    alphaBoolBlend {blendOption=Linear}
    boundBone {boneName=dyn_collar_front_r}
    componentAppliedLinearAccClamp=<1500 1500 1500>
    componentLinearAccScale=<0.4 0.4 1>
    componentLinearVelScale=<0.2 0.2 0.4>
    gravityOverride {z=-980}
    physicsBodyDefinitions [{
      sphereCollisionRadius=0
      boundBone {boneName=dyn_collar_front_r}
      boxExtents=<35 35 35>
      constraintSetup {angularLimitsMax=<5 0 5>, angularLimitsMin=<-5 -10 -5>, angularTarget=<1 0 0>}
      localJointOffset=<4 0 0>
     }]
    relativeSpaceBone {boneName=root}
   }
   n28 = Anim Dynamics - Bone: dyn_collar_front_l!(ComponentPose=@n27, Alpha=1, ExternalForce="0,0,0") -> (Pose) {
    angularDampingOverride=1
    angularSpringConstant=5000
    bAngularSpring=true
    bGravityOverrideInSimSpace=true
    bOverrideAngularDamping=true
    bUseGravityOverride=true
    simulationSpace=BoneRelative
    alphaBoolBlend {blendOption=Linear}
    boundBone {boneName=dyn_collar_front_l}
    componentAppliedLinearAccClamp=<1500 1500 1500>
    componentLinearAccScale=<0.4 0.4 1>
    componentLinearVelScale=<0.2 0.2 0.4>
    gravityOverride {z=-980}
    physicsBodyDefinitions [{
      sphereCollisionRadius=0
      boundBone {boneName=dyn_collar_front_l}
      boxExtents=<35 35 35>
      constraintSetup {angularLimitsMax=<5 0 5>, angularLimitsMin=<-5 -10 -5>, angularTarget=<1 0 0>}
      localJointOffset=<4 0 0>
     }]
    relativeSpaceBone {boneName=root}
   }
   n29 = Anim Dynamics - Bone: dyn_collar_front_r(ComponentPose=@n28, Alpha=1, ExternalForce="0,0,0") {
    angularDampingOverride=1
    angularSpringConstant=5000
    bAngularSpring=true
    bGravityOverrideInSimSpace=true
    bOverrideAngularDamping=true
    bUseGravityOverride=true
    simulationSpace=BoneRelative
    alphaBoolBlend {blendOption=Linear}
    boundBone {boneName=dyn_collar_front_r}
    componentAppliedLinearAccClamp=<1500 1500 1500>
    componentLinearAccScale=<0.4 0.4 1>
    componentLinearVelScale=<0.2 0.2 0.4>
    gravityOverride {z=-980}
    physicsBodyDefinitions [{
      sphereCollisionRadius=0
      boundBone {boneName=dyn_collar_front_r}
      boxExtents=<35 35 35>
      constraintSetup {angularLimitsMax=<2.5 0 2.5>, angularLimitsMin=<-2.5 -10 -2.5>, angularTarget=<1 0 0>}
      localJointOffset=<4 0 0>
     }]
    relativeSpaceBone {boneName=root}
   }
   n30 = Anim Dynamics - Bone: dyn_collar_front_l(ComponentPose=@n29, Alpha=1, ExternalForce="0,0,0") {
    angularDampingOverride=1
    angularSpringConstant=5000
    bAngularSpring=true
    bGravityOverrideInSimSpace=true
    bOverrideAngularDamping=true
    bUseGravityOverride=true
    simulationSpace=BoneRelative
    alphaBoolBlend {blendOption=Linear}
    boundBone {boneName=dyn_collar_front_l}
    componentAppliedLinearAccClamp=<1500 1500 1500>
    componentLinearAccScale=<0.4 0.4 1>
    componentLinearVelScale=<0.2 0.2 0.4>
    gravityOverride {z=-980}
    physicsBodyDefinitions [{
      sphereCollisionRadius=0
      boundBone {boneName=dyn_collar_front_l}
      boxExtents=<35 35 35>
      constraintSetup {angularLimitsMax=<2.5 0 2.5>, angularLimitsMin=<-2.5 -10 -2.5>, angularTarget=<1 0 0>}
      localJointOffset=<4 0 0>
     }]
    relativeSpaceBone {boneName=root}
   }
   n31 = Anim Dynamics - Bone: dyn_pin_l!(ComponentPose=@n30, Alpha=1, GravityOverride:Vector="0,0,-980", ExternalForce="0,0,-980") -> (Pose) {
    angularDampingOverride=0.9999
    angularSpringConstant=100
    bOverrideAngularDamping=true
    bUseGravityOverride=true
    simulationSpace=BoneRelative
    alphaBoolBlend {blendOption=Linear}
    boundBone {boneName=dyn_pin_l}
    componentAppliedLinearAccClamp=<1500 1500 1500>
    componentLinearAccScale=<0.4 0.4 0.5>
    componentLinearVelScale=<0.4 0.4 0.5>
    physicsBodyDefinitions [{
      boundBone {boneName=dyn_pin_l}
      boxExtents=<5 5 5>
      constraintSetup {angularLimitsMax=<5 10 20>, angularLimitsMin=<-5 -10 -20>, angularTarget=<0 0 0>}
      localJointOffset=<1 0 0>
     }]
    relativeSpaceBone {boneName=root}
   }
   n32 = Anim Dynamics - Bone: dyn_pin_r!(ComponentPose=@n31, Alpha=1, GravityOverride:Vector="0,0,-980", ExternalForce="0,0,-980") -> (Pose) {
    angularDampingOverride=0.9999
    angularSpringConstant=100
    bOverrideAngularDamping=true
    bUseGravityOverride=true
    simulationSpace=BoneRelative
    alphaBoolBlend {blendOption=Linear}
    boundBone {boneName=dyn_pin_r}
    componentAppliedLinearAccClamp=<1500 1500 1500>
    componentLinearAccScale=<0.4 0.4 0.5>
    componentLinearVelScale=<0.4 0.4 0.5>
    physicsBodyDefinitions [{
      boundBone {boneName=dyn_pin_r}
      boxExtents=<5 5 5>
      constraintSetup {angularLimitsMax=<5 10 20>, angularLimitsMin=<-5 -10 -20>, angularTarget=<0 0 0>}
      localJointOffset=<1 0 0>
     }]
    relativeSpaceBone {boneName=root}
   }
   n33 = Anim Dynamics - Bone: dyn_grenade_pin_fr!(ComponentPose=@n32, Alpha=1, GravityOverride:Vector="0,0,-980", ExternalForce="0,0,0") -> (Pose) {
    angularDampingOverride=0.99
    angularSpringConstant=100
    bOverrideAngularDamping=true
    bUseGravityOverride=true
    simulationSpace=BoneRelative
    alphaBoolBlend {blendOption=Linear}
    boundBone {boneName=dyn_grenade_pin_fr}
    componentAppliedLinearAccClamp=<1500 1500 1500>
    componentLinearAccScale=<0.8 0.8 0.9>
    componentLinearVelScale=<0.8 0.8 0.9>
    physicsBodyDefinitions [{
      boundBone {boneName=dyn_grenade_pin_fr}
      boxExtents=<3.5 3.5 3.5>
      constraintSetup {angularLimitsMax=<2 0 10>, angularLimitsMin=<-2 -90 -10>, angularTarget=<0 0 0>}
      localJointOffset=<0.7 0 0>
     }]
    relativeSpaceBone {boneName=root}
   }
   n34 = Anim Dynamics - Bone: dyn_grenade_pin_bk!(ComponentPose=@n33, Alpha=1, GravityOverride:Vector="0,0,-980", ExternalForce="0,0,0") -> (Pose) {
    angularDampingOverride=0.99
    angularSpringConstant=100
    bOverrideAngularDamping=true
    bUseGravityOverride=true
    simulationSpace=BoneRelative
    alphaBoolBlend {blendOption=Linear}
    boundBone {boneName=dyn_grenade_pin_bk}
    componentAppliedLinearAccClamp=<1500 1500 1500>
    componentLinearAccScale=<0.8 0.8 0.9>
    componentLinearVelScale=<0.8 0.8 0.9>
    physicsBodyDefinitions [{
      boundBone {boneName=dyn_grenade_pin_bk}
      boxExtents=<3.5 3.5 3.5>
      constraintSetup {angularLimitsMax=<2 0 10>, angularLimitsMin=<-2 -90 -10>, angularTarget=<0 0 0>}
      localJointOffset=<0.7 0 0>
     }]
    relativeSpaceBone {boneName=root}
   }
   n35 = "Transform (Modify) Bone - Bone: dyn_jacket_l_master"!(ComponentPose=@n34, bAlphaBoolEnabled:bool=True, Translation="0,0,0", Rotation="10,0,0", Scale="1,1,1") -> (Pose) {
    alphaInputType=Bool
    rotationMode=BMM_Additive
    rotationSpace=BCS_BoneSpace
    translationMode=BMM_Additive
    translationSpace=BCS_BoneSpace
    alphaBoolBlend {blendOption=Linear}
    boneToModify {boneName=dyn_jacket_l_master}
    rotation {pitch=10}
   }
   n36 = "Transform (Modify) Bone - Bone: dyn_jacket_l_bk_master"!(ComponentPose=@n35, bAlphaBoolEnabled:bool=True, Translation="0,0,0", Rotation="10,0,0", Scale="1,1,1") -> (Pose) {
    alphaInputType=Bool
    rotationMode=BMM_Additive
    rotationSpace=BCS_BoneSpace
    translationMode=BMM_Additive
    translationSpace=BCS_BoneSpace
    alphaBoolBlend {blendOption=Linear}
    boneToModify {boneName=dyn_jacket_l_bk_master}
    rotation {pitch=10}
   }
   n37 = "Transform (Modify) Bone - Bone: dyn_jacket_l_fr_master"!(ComponentPose=@n36, bAlphaBoolEnabled:bool=True, Translation="0,0,0", Rotation="10,0,0", Scale="1,1,1") -> (Pose) {
    alphaInputType=Bool
    rotationMode=BMM_Additive
    rotationSpace=BCS_BoneSpace
    translationMode=BMM_Additive
    translationSpace=BCS_BoneSpace
    alphaBoolBlend {blendOption=Linear}
    boneToModify {boneName=dyn_jacket_l_fr_master}
    rotation {pitch=10}
   }
   n38 = Get back_constrain() -> (back_constrain:double)
   n39 = "Constraint - dyn_jacket_l_bk_02 (P)"(ComponentPose=@n37, Alpha=@n38, ConstraintWeights_0=1) {
    alphaBoolBlend {blendOption=Linear}
    boneToModify {boneName=dyn_jacket_l_bk_02}
    constraintSetup [{targetBone {boneName=pelvis}}]
    constraintWeights [1]
   }
   n40 = "Constraint - dyn_jacket_r_bk_02 (P)"(ComponentPose=@n39, Alpha=@n38, ConstraintWeights_0=1) {
    alphaBoolBlend {blendOption=Linear}
    boneToModify {boneName=dyn_jacket_r_bk_02}
    constraintSetup [{targetBone {boneName=pelvis}}]
    constraintWeights [1]
   }
   Output Pose(Result=Component To Local(ComponentPose=@n40)) {layerGroup=None}
  ]}
 EventGraph {nodes [
   Event BlueprintUpdateAnimation()
   > n1 = Sequence()
   > n2 = Gravity Override State Function DEPRECATED(
    CharacterPartAnimInstance=Self-Reference()
    GameTimeInSeconds=GetGameTimeInSeconds()
    GravityOverrideParams=Get GravityOverrideParams() -> (GravityOverrideParams:Gravity Override Params Struct Structure)
   )
   > Set GravityOverride(GravityOverride=@n2)
   n3 = Get bIsSkydiveDiveMode()
   Set back_orient_out(
    execute=@n1.then_1
    back_orient_out:double=FInterpTo(Current=Get back_orient_out(), Target=SelectFloat(A=1, B=0, bPickA=@n3), DeltaTime=0.5, InterpSpeed=0.6)
   ) -> (Output_Get:double)
   n4 = AND Boolean(A=Get bIsSkydiveFloating(), B=Get bRigidBodyEnabled())
   > Set back_constrain(back_constrain:double=FInterpTo(Current=Get back_constrain(), Target=SelectFloat(A=0.25, B=0, bPickA=@n4), DeltaTime=0.5, InterpSpeed=0.6)) -> (Output_Get:double)
   n5 = OR Boolean(A=Get bIsCrouching(), B=NOT Boolean(A=Get bRigidBodyEnabled()))
   > Set crouch_alpha(crouch_alpha:double=FInterpTo(Current=Get crouch_alpha(), Target=SelectFloat(A=1, B=0, bPickA=@n5), DeltaTime=0.5, InterpSpeed=0.4)) -> (Output_Get:double)
   > Set emote alpha(emote alpha:double=SelectFloat(A=0, B=1, bPickA=Get bIsPlayingEmote())) -> (Output_Get:double)
   n6 = AND Boolean(A=NOT Boolean(A=Get bRigidBodyEnabled()), B=Get bIsCrouching())
   > Set back_in(back_in:double=FInterpTo(Current=Get back_in(), Target=SelectFloat(A=1, B=0, bPickA=@n6), DeltaTime=0.5, InterpSpeed=0.4)) -> (Output_Get:double)
   OR Boolean(A=Get bIsSkydiveFloating(), B=@n3)
  ]}
]
variables [
 back_constrain {type=number}
 back_in {type=number}
 back_orient_out {type=number}
 crouch_alpha {type=number, default=1}
 emote alpha {type=number, default=1}
 GravityOverride {type=Vector}
 GravityOverrideParams {type=GravityOverrideParamsStruct}
]
